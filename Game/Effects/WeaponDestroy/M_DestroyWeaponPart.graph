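# M_DestroyWeaponPart is a simple Default Lit material for mesh particles, apparently for weapon parts in the weapon destroy effect, judging by its name and folder.
# It holds only constant nodes, a near-black color and two scalars, with no textures or parameters.
# The file does not show which material inputs these constants feed.
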Material /Game/Effects/WeaponDestroy/M_DestroyWeaponPart
graph {
 properties {bUsedWithMeshParticles=true, pixelDepthOffsetMode=PDOM_Legacy, usedShadingModels=Default Lit}
 nodes [
  Constant4Vector() {constant=<rgba 0.04 0.04 0.04 1>}
  Constant() {r=1}
  Constant() {r=0.2}
 ]
}
