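# Unlit translucent material for mesh particles, named as the star particle of the weapon-destroy effect.
# Emissive blends from raw particle color to color tinted by a Color parameter, then passes through MF_FX_EmissiveAdjustments.
# The same SimpleColorSwitch parameter blends particle alpha between a depth-faded version and one scaled by AlphaScalar.
# A quality switch drops the depth fade at low quality.
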Material /Game/Effects/WeaponDestroy/M_StarParticle
graph {
 properties {
  bUsedWithMeshParticles=true
  bUsedWithNiagaraMeshParticles=true
  blendMode=BLEND_Translucent
  pixelDepthOffsetMode=PDOM_Legacy
  shadingModel=MSM_Unlit
  usedShadingModels=Unlit
 }
 nodes [
  n1 = ParticleColor()
  n2 = DepthFade(FadeDistance=Constant() {r=12})
  LinearInterpolate(
   A=QualitySwitch(Default=Multiply(A=@n1.A, B=@n2), Low=@n1.A)
   B=Multiply(A=@n1.A, B=ScalarParameter() {defaultValue=1, parameterName=AlphaScalar})
   Alpha=ScalarParameter() {parameterName=SimpleColorSwitch, sliderMax=1}
  )
  n3 = VectorParameter() {parameterName=Color, defaultValue=<rgba 1 1 1 1>}
  MaterialFunctionCall(Emissive=LinearInterpolate(A=@n1.RGB, B=Multiply(A=@n3.RGB, B=@n1.RGB), Alpha=ScalarParameter() {parameterName=SimpleColorSwitch, sliderMax=1})) {materialFunction=&MF_FX_EmissiveAdjustments}
 ]
}
