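# Translucent unlit particle material for a weapon-destroy effect, drawing an icon texture (pistol by default) tinted by particle color and alpha.
# A RotateBlend dynamic parameter scales a vertex offset chosen by the UseShatter static switch.
# Shatter derives the offset from centered UV1, feeding only U into both X and Y; otherwise vertex red blends two opposing camera-axis rotations with view-space shifts.
# A DebugTimeSine multiply of the shatter offset appears to be leftover preview logic.
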Material /Game/Effects/WeaponDestroy/M_WeaponDestroyExclamation
graph {
 properties {
  bUsedWithMeshParticles=true
  bUsedWithNanite=true
  bUsedWithNiagaraMeshParticles=true
  bUsedWithNiagaraSprites=true
  blendMode=BLEND_Translucent
  pixelDepthOffsetMode=PDOM_Legacy
  shadingModel=MSM_Unlit
  usedShadingModels=Unlit
 }
 nodes [
  n1 = ParticleColor()
  MaterialFunctionCall(Emissive=@n1.RGB) {materialFunction=&MF_FX_EmissiveAdjustments}
  Multiply(A=@n1.A, B=TextureSampleParameter2D() {parameterName=IconTex, texture=/Game/UI/Foundation/Textures/Icons/ItemTypes/T-Icon-Pistol-128}.A)
  n2 = TextureCoordinate() {coordinateIndex=1}
  n3 = MaterialFunctionCall(Float2=Multiply(A=Subtract(A=@n2) {constB=0.5}, B=ScalarParameter() {defaultValue=50, parameterName=ShatterDistance}) {constB=7}) {materialFunction=&BreakOutFloat2Components}
  n4 = Transform(None=MaterialFunctionCall(X=@n3.R, Y=@n3.R) {materialFunction=&MakeFloat3}.Result)
  n5 = Constant3Vector() {constant=<rgba 1 1 0 0>}
  n6 = Add(
   A=RotateAboutAxis(
    NormalizedRotationAxis=MaterialFunctionCall() {materialFunction=&CameraDirectionVector}.Result
    RotationAngle=Constant() {r=0.05}
    PivotPoint=ObjectPositionWS()
    Position=WorldPosition().XYZ
   )
   B=Multiply(A=Transform(None=@n5) {transformSourceType=TRANSFORMSOURCE_View}) {constB=-3}
  )
  n7 = Constant3Vector() {constant=<rgba 1 1 0 0>}
  n8 = Add(
   A=RotateAboutAxis(
    NormalizedRotationAxis=MaterialFunctionCall() {materialFunction=&CameraDirectionVector}.Result
    RotationAngle=Constant() {r=-0.15}
    PivotPoint=ObjectPositionWS()
    Position=WorldPosition().XYZ
   )
   B=Multiply(A=Transform(None=@n7) {transformSourceType=TRANSFORMSOURCE_View}) {constB=5}
  )
  n9 = VertexColor()
  Multiply(
   A=DynamicParameter() {defaultValue=<rgba 0 1 1 1>, paramNames [RotateBlend,Param2,Param3,Param4]}.RotateBlend
   B=StaticSwitchParameter(True=@n4, False=LinearInterpolate(A=@n6, B=@n8, Alpha=@n9.R)) {parameterName=UseShatter}
  )
  Multiply(A=@n4, B=MaterialFunctionCall() {materialFunction=&DebugTimeSine}.Result)
 ]
}
