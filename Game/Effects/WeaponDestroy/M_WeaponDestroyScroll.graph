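# Unlit translucent mesh-particle material that draws a scrolling streak for the weapon destroy effect.
# Particle relative time shifts the streak texture's V offset from -1 to 1, sweeping it once across the mesh over each particle's life.
# That sample is multiplied by particle alpha and the unscrolled V coordinate as a linear fade.
# Particle color RGB feeds the shared MF_FX_EmissiveAdjustments function for emissive.
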Material /Game/Effects/WeaponDestroy/M_WeaponDestroyScroll
graph {
 properties {
  bUsedWithMeshParticles=true
  blendMode=BLEND_Translucent
  pixelDepthOffsetMode=PDOM_Legacy
  previewMesh=&/SM_SquiggleSimple
  shadingModel=MSM_Unlit
  usedShadingModels=Unlit
 }
 nodes [
  n1 = ParticleColor()
  MaterialFunctionCall(Emissive=@n1.RGB) {materialFunction=&MF_FX_EmissiveAdjustments}
  n2 = Constant() {r=-1}
  n3 = Multiply(A=ParticleRelativeTime()) {constB=2}
  n4 = Add(
   A=TextureCoordinate()
   B=MaterialFunctionCall(X=Constant(), Y=Add(A=@n2, B=@n3)) {materialFunction=&MakeFloat2}.Result
  )
  Multiply(
   A=Multiply(A=@n1.A, B=TextureSample(UVs=@n4) {texture=&/T_WeaponDestroy_Streak}.RGB)
   B=ComponentMask(None=TextureCoordinate()) {g=true}
  )
 ]
}
